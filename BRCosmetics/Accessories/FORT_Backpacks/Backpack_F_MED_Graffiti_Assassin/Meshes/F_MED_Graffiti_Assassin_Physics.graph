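# Physics asset for the Graffiti Assassin backpack cosmetic, giving its dangling parts secondary motion.
# A box on the pack bone anchors box bodies for the flap, two zippers and strap segments, plus a capsule for the sword.
# Flap, zippers and sword constrain directly to the pack, while each side's straps form two separate two-segment chains rooted on it.
# Narrow swing and twist limits keep every piece close to its rest pose.
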PhysicsAsset /BRCosmetics/Accessories/FORT_Backpacks/Backpack_F_MED_Graffiti_Assassin/Meshes/F_MED_Graffiti_Assassin_Physics
bodies {
 dyn_flap [{shapeType=Box, extentX=2.7116, extentY=10.4839, extentZ=1.4697, center=<1.8988 -0.0157 -0.1666>}]
 dyn_strap_l_01 [{shapeType=Box, extentX=2.0325, extentY=1.3074, extentZ=0.7301, center=<0.6263 -0.0191 0.3451>}]
 dyn_strap_l_02 [{shapeType=Box, extentX=2.0325, extentY=1.0892, extentZ=0.8384, center=<0.9733 0.511 -0.0545>}]
 dyn_strap_l_03 [{shapeType=Box, extentX=2.0325, extentY=1.1927, extentZ=0.7301, center=<0.7947 0 -0.2305>}]
 dyn_strap_l_04 [{shapeType=Box, extentX=2.0325, extentY=1.1766, extentZ=0.9931, center=<1.1077 -0.1435 -0.0277>}]
 dyn_strap_r_01 [{shapeType=Box, extentX=2.0325, extentY=1.01, extentZ=0.7301, center=<0.2147 0.3315 -0.3639>}]
 dyn_strap_r_02 [{shapeType=Box, extentX=2.0325, extentY=1.1091, extentZ=0.9401, center=<1.2005 0.3662 0.2083>}]
 dyn_strap_r_03 [{shapeType=Box, extentX=2.0325, extentY=1.01, extentZ=0.7301, center=<0.8196 0.377 -0.0002>}]
 dyn_strap_r_04 [{shapeType=Box, extentX=2.0325, extentY=1.2306, extentZ=1.1649, center=<0.9733 -0.0517 0.1774>}]
 dyn_sword [{shapeType=Capsule, length=23.377, radius=4.4949, center=<1 0 0>, rotation=<rot 90 0 0>}]
 dyn_zipper_l [{shapeType=Box, extentX=1.8455, extentY=1.01, extentZ=0.5, center=<0.44 0.0163 -0.0225>}]
 dyn_zipper_r [{shapeType=Box, extentX=2.3131, extentY=1.01, extentZ=0.5922, center=<1.0115 0 0.0143>}]
 pack [{shapeType=Box, extentX=23.546, extentY=12.1451, extentZ=4.8935, center=<-0.5944 -6.0938 -12.8452>}]
}
constraints [
 {bone1Name=dyn_sword, bone2Name=pack, swing1LimitDegrees=10, swing2LimitDegrees=10, twistLimitDegrees=5}
 {bone1Name=dyn_zipper_r, bone2Name=pack, swing1LimitDegrees=30, swing2LimitDegrees=30, twistLimitDegrees=5}
 {bone1Name=dyn_zipper_l, bone2Name=pack, swing1LimitDegrees=30, swing2LimitDegrees=30, twistLimitDegrees=5}
 {bone1Name=dyn_flap, bone2Name=pack, swing1LimitDegrees=15, swing2LimitDegrees=15, twistLimitDegrees=2}
 {bone1Name=dyn_strap_r_01, bone2Name=pack, swing1LimitDegrees=20, swing2LimitDegrees=20, twistLimitDegrees=5}
 {bone1Name=dyn_strap_r_02, bone2Name=dyn_strap_r_01, swing1LimitDegrees=30, swing2LimitDegrees=30, twistLimitDegrees=5}
 {bone1Name=dyn_strap_r_03, bone2Name=pack, swing1LimitDegrees=30, swing2LimitDegrees=30, twistLimitDegrees=5}
 {bone1Name=dyn_strap_r_04, bone2Name=dyn_strap_r_03, swing1LimitDegrees=15, swing2LimitDegrees=30, twistLimitDegrees=5}
 {bone1Name=dyn_strap_l_01, bone2Name=pack, swing1LimitDegrees=10, swing2LimitDegrees=15, twistLimitDegrees=5}
 {bone1Name=dyn_strap_l_02, bone2Name=dyn_strap_l_01, swing1LimitDegrees=15, swing2LimitDegrees=30, twistLimitDegrees=5}
 {bone1Name=dyn_strap_l_03, bone2Name=pack, swing1LimitDegrees=15, swing2LimitDegrees=30, twistLimitDegrees=5}
 {bone1Name=dyn_strap_l_04, bone2Name=dyn_strap_l_03, swing1LimitDegrees=15, swing2LimitDegrees=30, twistLimitDegrees=5}
]
