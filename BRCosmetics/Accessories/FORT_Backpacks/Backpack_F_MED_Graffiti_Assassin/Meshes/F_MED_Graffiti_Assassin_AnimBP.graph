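# Animation Blueprint for the Graffiti Assassin backpack cosmetic, adding secondary physics motion to its skeleton.
# A single RigidBody node simulates the reference pose with an override physics asset, in base-bone space relative to root, and feeds the output pose.
# It uses overridden world gravity, scaled component velocity and acceleration, and the skydiving force vector from the Fortnite anim input as external force.
# The event graph holds only unconnected stub nodes.
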AnimBlueprint /BRCosmetics/Accessories/FORT_Backpacks/Backpack_F_MED_Graffiti_Assassin/Meshes/F_MED_Graffiti_Assassin_AnimBP
target_skeleton=&/F_MED_Graffiti_Assassin_Skeleton
graphs [
 AnimGraph {nodes [
   n1 = RigidBody(
    ComponentPose=Local To Component(LocalPose=Local Space Ref Pose())
    ExternalForce=Break Fort Anim Input Skydiving External Force(FortAnimInput_SkydivingExternalForce=Get SkydivingExternalForceInput()) -> (
     bUseSkydivingVectorForce=~
     SkydivingVectorForce=~
     HeadToPelvisDirection=~
     FloatingMultiplier=~
     FloatingAdditive=~
     DivingMultiplier=~
     DivingAdditive=~
     ParachutingMultiplier=~
     ParachutingAdditive=~
    ).SkydivingVectorForce
   ) {
    bOverrideWorldGravity=true
    overridePhysicsAsset=&/F_MED_Graffiti_Assassin_Physics
    simulationSpace=BaseBoneSpace
    alphaBoolBlend {blendOption=Linear}
    baseBoneRef {boneName=root}
    componentLinearAccScale=<0.4 0.3 0.3>
    componentLinearVelScale=<0.4 0.3 0.3>
    overrideWorldGravity {z=-980}
   }
   Output Pose(Result=Component To Local(ComponentPose=@n1)) {layerGroup=None}
  ]}
 EventGraph {nodes [
   TryGetPawnOwner()
   Event BlueprintUpdateAnimation()
  ]}
]
variables []
